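; Audio component for a GrillMoon gold ribbon spline actor that plays the ribbon sound at random points along the ribbon.
; BeginPlay casts the owner to BP_RibbonsSpline_GrillMoon and starts a looping timer whose interval is randomized only once, at start.
; Each tick of that timer plays the sound at a random world location along the owner's spline, if valid.
; EndPlay clears the timer.
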(event EventBeginPlay
  (bind _asbp_ribbons_spline_grill_moon (Utilities|Casting|CastToBP_RibbonsSpline_GrillMoon (Components|GetOwner)))
  (Variables|Default|SetRibbon%20Spline _asbp_ribbons_spline_grill_moon)
  (bind _returnvalue (Utilities|Time|SetTimerbyEvent (delegate Custom|Play%20Gold%20Ribbon%20Sound) (Math|Random|RandomFloatinRange 3.0 6.0) true))
  (Variables|Default|SetRibbon%20Sound%20Timer _returnvalue))

(event Custom|Play%20Gold%20Ribbon%20Sound
  (bind _ribbon_spline (Variables|Default|GetRibbon%20Spline)
    (:then
      (bind _spline (Variables|/GrillMoon/Assets/TechArt/GoldRibbons/BP_RibbonsSpline_GrillMoon.BP_RibbonsSpline_GrillMoon_C|GetSpline _ribbon_spline)
        (:then
          (Audio|PlaySoundatLocation (Variables|Default|GetGold%20Ribbon%20Sound) (Spline|GetLocationatDistanceAlongSpline _spline (Math|Random|RandomFloatinRange 0.0 (Spline|GetSplineLength _spline)) "World")))
        (:else)))
    (:else)))

(event EventEndPlay
  (params (EndPlayReason :type EEndPlayReason%20Enum :default "Destroyed"))
  (Utilities|Time|ClearandInvalidateTimerbyHandle (Variables|Default|GetRibbon%20Sound%20Timer)))
