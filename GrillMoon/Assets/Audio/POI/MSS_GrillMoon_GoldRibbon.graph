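# MetaSound source for a spatialized mono one-shot sparkle sound, named for the Gold Ribbon in GrillMoon's POI audio.
# On play, it picks a random wave from three borrowed glider sparkle-flap variations and feeds it to a wave player.
# A MasterVolume input scales the output, and the wave player's finish trigger ends the sound.
# It uses Gold Ribbon attenuation, the events sound class, and a concurrency override that blocks new instances at the limit.
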MetaSoundSource /GrillMoon/Assets/Audio/POI/MSS_GrillMoon_GoldRibbon
inputs [
 MasterVolume {type=Float}
 Sounds {type=WaveAsset:Array}
 UE.Source.OnPlay {type=Trigger}
]
interfaces [UE.OutputFormat.Mono:1.0, UE.Source.OneShot:1.0, UE.Source:1.1]
outputs [
 "UE.OutputFormat.Mono.Audio:0" {type=Audio}
 UE.Source.OneShot.OnFinished {type=Trigger}
]
pages [Default {
  input_defaults {
   MasterVolume=0.15
   UE.Source.OnPlay=null
   Sounds [
    /Game/Sounds/Fort_Gadgets/Athena/Glider_Muertos_Flap_Sparkle_01
    /Game/Sounds/Fort_Gadgets/Athena/Glider_Muertos_Flap_Sparkle_02
    /Game/Sounds/Fort_Gadgets/Athena/Glider_Muertos_Flap_Sparkle_03
   ]
  }
  nodes [
   ArrayRandomGet_0 = Array.Random Get:WaveAsset:Array(In Array=^Sounds, Next=^UE.Source.OnPlay)
   WavePlayer_0 = UE.Wave Player:Mono(Play=@ArrayRandomGet_0.On Next, Wave Asset=@ArrayRandomGet_0.Value)
   Multiply_0 = UE.Multiply:Audio by Float(AdditionalOperands=^MasterVolume, PrimaryOperand=@WavePlayer_0.Out Mono)
  ]
  links [{dst="@UE.OutputFormat.Mono.Audio:0", src=Multiply_0.Out}, {dst="@UE.Source.OneShot.OnFinished", src=WavePlayer_0.On Finished}]
 }]
properties {
 assetImportData=&self:AssetImportData
 attenuationSettings=/GrillMoon/Assets/Audio/Attenuation/ATT_GrillMoon_GoldRibbon
 bOverrideConcurrency=true
 maxDistance=3700
 sampleRate=48000
 soundClassObject=&Fort_Events
 concurrencyOverrides {maxCount=5, resolutionRule=PreventNew}
}
