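# Mono MetaSound source for the GrillMoon radio point of interest, playing an ambient radio loop attenuated with its own POI radio settings.
# OnPlay starts a wave player on the Sound input, whose output is scaled by MasterVolume before reaching the mono out.
# The source is flagged looping, keeps playing when silent, and overrides concurrency to a single instance.
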MetaSoundSource /GrillMoon/Assets/Audio/POI/MSS_GrillMoon_POI_Radio
inputs [
 MasterVolume {type=Float}
 Sound {type=WaveAsset}
 UE.Source.OnPlay {type=Trigger}
]
interfaces [UE.OutputFormat.Mono:1.0, UE.Source:1.1]
outputs ["UE.OutputFormat.Mono.Audio:0" {type=Audio}]
pages [Default {
  input_defaults {MasterVolume=1, Sound=/GrillMoon/Assets/Audio/wav/GrillMoon_Radio_Amb_Loop, UE.Source.OnPlay=null}
  nodes [
   WavePlayer_0 = UE.Wave Player:Mono(Wave Asset=^Sound, Play=^UE.Source.OnPlay) {also {defaults {Loop="<MetaSoundBuilderResult.SUCCEEDED: 0>"}}}
   Multiply_0 = UE.Multiply:Audio by Float(AdditionalOperands=^MasterVolume, PrimaryOperand=@WavePlayer_0.Out Mono)
  ]
  links [{dst="@UE.OutputFormat.Mono.Audio:0", src=Multiply_0.Out}]
 }]
properties {
 assetImportData=&self:AssetImportData
 attenuationSettings=/GrillMoon_Audio/Assets/Settings/ATT_GrillMoon_POI_Radio
 bLooping=true
 bOverrideConcurrency=true
 duration=10000
 maxDistance=700
 priority=0.2
 sampleRate=48000
 soundClassObject=&Fort_Events
 virtualizationMode=PlayWhenSilent
 concurrencyOverrides {maxCount=1}
}
