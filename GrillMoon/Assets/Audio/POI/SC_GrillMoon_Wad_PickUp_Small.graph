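# Sound cue for picking up a small cash wad in GrillMoon, routed to the loot-collect sound class at reduced volume.
# A mixer layers a random cash pickup sound with a slightly quieter, briefly delayed random tone.
# Each layer draws from high, medium and low variants through its own modulator for variation.
# Playback uses shared pickup attenuation, and a concurrency override caps it at two instances, stopping the oldest.
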SoundCue /GrillMoon/Assets/Audio/POI/SC_GrillMoon_Wad_PickUp_Small
root=Mixer_0
nodes [
 {id=Delay_0, children [Modulator_0], properties {delayMax=0.1, delayMin=0.1}}
 {id=Mixer_0, children [Modulator_1,Delay_0], properties {inputVolume [1,0.9]}}
 {id=Modulator_0, children [Random_2], properties {pitchMax=1.03, pitchMin=0.97}}
 {id=Modulator_1, children [Random_0]}
 {id=Random_0, children [WavePlayer_4,WavePlayer_5,WavePlayer_7], properties {weights [1,1,1]}}
 {id=Random_2, children [WavePlayer_8, WavePlayer_9, WavePlayer_10], properties {weights [1,1,1]}}
 {id=WavePlayer_10, properties {soundWaveAssetPtr=/Game/Athena/Sounds/Items/Wads/Cash_PickUp_Tone_Med}}
 {id=WavePlayer_4, properties {soundWaveAssetPtr=/Game/Athena/Sounds/Items/Wads/Cash_Pickup_High}}
 {id=WavePlayer_5, properties {soundWaveAssetPtr=/Game/Athena/Sounds/Items/Wads/Cash_Pickup_Low}}
 {id=WavePlayer_7, properties {soundWaveAssetPtr=/Game/Athena/Sounds/Items/Wads/Cash_Pickup_Med}}
 {id=WavePlayer_8, properties {soundWaveAssetPtr=/Game/Athena/Sounds/Items/Wads/Cash_PickUp_Tone_High}}
 {id=WavePlayer_9, properties {soundWaveAssetPtr=/Game/Athena/Sounds/Items/Wads/Cash_PickUp_Tone_Low}}
]
properties {
 attenuationSettings=&Fort_Weapon_Pickup_Att
 bOverrideConcurrency=true
 maxDistance=3500
 soundClassObject=&Fort_Loot_Collect
 volumeMultiplier=0.25
 attenuationOverrides {falloffDistance=350, nonSpatializedRadiusStart=200, stereoSpread=512, attenuationShapeExtents {x=150}}
 concurrencyOverrides {maxCount=2, resolutionRule=StopOldest}
 concurrencySet [null]
}
